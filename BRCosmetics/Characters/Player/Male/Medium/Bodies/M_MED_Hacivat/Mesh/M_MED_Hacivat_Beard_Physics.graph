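# Physics asset for the beard mesh of the M_MED_Hacivat male player cosmetic body, defining collision and joint limits for its rope bones.
# Two three-link chains, dyn_rope A and B, use small capsules, each linked to its parent and rooted on a larger head capsule.
# Every joint shares the same moderate swing cone with near-zero twist, so the strands can bend but barely spin.
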PhysicsAsset /BRCosmetics/Characters/Player/Male/Medium/Bodies/M_MED_Hacivat/Mesh/M_MED_Hacivat_Beard_Physics
bodies {
 dyn_rope_01_A [{shapeType=Capsule, length=3, radius=1, center=<1.5 0 0>, rotation=<rot 90 0 0>}]
 dyn_rope_01_B [{shapeType=Capsule, length=3, radius=1, center=<1.5 0 0>, rotation=<rot 90 0 0>}]
 dyn_rope_02_A [{shapeType=Capsule, length=3, radius=1, center=<2 0 0>, rotation=<rot 90 0 0>}]
 dyn_rope_02_B [{shapeType=Capsule, length=3, radius=1, center=<2 0 0>, rotation=<rot 90 0 0>}]
 dyn_rope_03_A [{shapeType=Capsule, length=3, radius=1, center=<3 0 0>, rotation=<rot 90 0 0>}]
 dyn_rope_03_B [{shapeType=Capsule, length=3, radius=1, center=<3 0 0>, rotation=<rot 90 0 0>}]
 head [{shapeType=Capsule, length=7.1525, radius=7.1358, center=<2.3915 1.4563 0.2726>, rotation=<rot -89.5227 -89.3633 89.8368>}]
}
constraints [
 {bone1Name=dyn_rope_01_B, bone2Name=head, swing1LimitDegrees=25, swing2LimitDegrees=25, twistLimitDegrees=1}
 {bone1Name=dyn_rope_02_B, bone2Name=dyn_rope_01_B, swing1LimitDegrees=25, swing2LimitDegrees=25, twistLimitDegrees=1}
 {bone1Name=dyn_rope_03_B, bone2Name=dyn_rope_02_B, swing1LimitDegrees=25, swing2LimitDegrees=25, twistLimitDegrees=1}
 {bone1Name=dyn_rope_01_A, bone2Name=head, swing1LimitDegrees=25, swing2LimitDegrees=25, twistLimitDegrees=1}
 {bone1Name=dyn_rope_02_A, bone2Name=dyn_rope_01_A, swing1LimitDegrees=25, swing2LimitDegrees=25, twistLimitDegrees=1}
 {bone1Name=dyn_rope_03_A, bone2Name=dyn_rope_02_A, swing1LimitDegrees=25, swing2LimitDegrees=25, twistLimitDegrees=1}
]
