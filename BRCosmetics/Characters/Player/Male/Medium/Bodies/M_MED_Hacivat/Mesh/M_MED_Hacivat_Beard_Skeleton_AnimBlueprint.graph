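# Post-copy AnimBlueprint for the Hacivat character's beard and moustache, adding secondary motion to the parent mesh's pose.
# The copied pose gets a beard pitch driven by jaw-to-spine rotation, then Anim Dynamics springs moustache, beard and rope bones.
# Playlists using custom character parts instead output the pose with the head shrunk to near zero.
# Skydive and parachute states set a Perlin-noise force for a RigidBody node left unwired.
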AnimBlueprint /BRCosmetics/Characters/Player/Male/Medium/Bodies/M_MED_Hacivat/Mesh/M_MED_Hacivat_Beard_Skeleton_AnimBlueprint
target_skeleton=/Game/Characters/Player/Male/Medium/Bodies/M_MED_Hacivat/Mesh/M_MED_Hacivat_Beard_stache_Skeleton
graphs [
 AnimGraph {nodes [
   RigidBody(ExternalForce=Get rigidbody_external_force() -> (rigidbody_external_force:Vector)) {
    bOverrideWorldGravity=true
    overridePhysicsAsset=&/M_MED_Hacivat_Beard_Physics
    simulationSpace=BaseBoneSpace
    alphaBoolBlend {blendOption=Linear}
    baseBoneRef {boneName=neck_01}
    componentAppliedLinearAccClamp=<900 900 900>
    componentLinearAccScale=<0.3 0.3 0.35>
    componentLinearVelScale=<0.35 0.35 0.4>
    overrideWorldGravity {z=-980}
   }
   n1 = "Transform (Modify) Bone - Bone: head"!(ComponentPose=Local To Component(LocalPose=Use cached pose 'PreScale Cache'()), Alpha=1) -> (Pose) {scaleMode=BMM_Replace, alphaBoolBlend {blendOption=Linear}, boneToModify {boneName=head}, scale=<0.01 0.01 0.01>}
   n2 = Anim Dynamics - Bone: mustache_l(ComponentPose=Local To Component(LocalPose=Use cached pose 'PreScale Cache'()), Alpha=0.7, ExternalForce="0,0,0") {
    angularDampingOverride=0.8
    angularSpringConstant=10000
    bAngularSpring=true
    bGravityOverrideInSimSpace=true
    bOverrideAngularDamping=true
    numSolverIterationsPostUpdate=4
    numSolverIterationsPreUpdate=16
    alphaBoolBlend {blendOption=Linear}
    boundBone {boneName=mustache_l}
    componentLinearAccScale=<2 2>
    physicsBodyDefinitions [{
      sphereCollisionRadius=0
      boundBone {boneName=mustache_l}
      boxExtents=<30 30 20>
      constraintSetup {angularTargetAxis=AxisY, angularLimitsMax=<0 5 0>, angularLimitsMin=<0 -35 0>, angularTarget=<0 1 0>}
      localJointOffset=<2 0 0>
     }]
   }
   n3 = Anim Dynamics - Bone: mustache_r(ComponentPose=@n2, Alpha=0.7, ExternalForce="0,0,0") {
    angularDampingOverride=0.8
    angularSpringConstant=10000
    bAngularSpring=true
    bGravityOverrideInSimSpace=true
    bOverrideAngularDamping=true
    numSolverIterationsPostUpdate=4
    numSolverIterationsPreUpdate=16
    alphaBoolBlend {blendOption=Linear}
    boundBone {boneName=mustache_r}
    componentLinearAccScale=<2 2>
    physicsBodyDefinitions [{
      sphereCollisionRadius=0
      boundBone {boneName=mustache_r}
      boxExtents=<30 30 20>
      constraintSetup {angularTargetAxis=AxisY, angularLimitsMax=<0 5 0>, angularLimitsMin=<0 -35 0>, angularTarget=<0 1 0>}
      localJointOffset=<-2 0 0>
     }]
   }
   n4 = Anim Dynamics - Bone: dyn_beard_01(ComponentPose=@n3, Alpha=0.4, ExternalForce="0,0,0") {
    angularDampingOverride=0.9999
    angularSpringConstant=5000
    bAngularSpring=true
    bGravityOverrideInSimSpace=true
    bOverrideAngularDamping=true
    simulationSpace=BoneRelative
    alphaBoolBlend {blendOption=Linear}
    boundBone {boneName=dyn_beard_01}
    componentAppliedLinearAccClamp=<5000 5000 5000>
    componentLinearAccScale=<0.5 0.5 0.5>
    componentLinearVelScale=<0.5 0.5 0.5>
    physicsBodyDefinitions [{
      sphereCollisionRadius=0
      boundBone {boneName=dyn_beard_01}
      boxExtents=<10 20 20>
      constraintSetup {angularTargetAxis=AxisY, angularLimitsMax=<15 35 35>, angularLimitsMin=<-15 -35 -35>, angularTarget=<0 1 0>}
      localJointOffset=<1 0 0>
     }]
   }
   n5 = Anim Dynamics - Bone: dyn_rope_02_B(ComponentPose=@n4, Alpha=1, ExternalForce="0,0,0") {
    angularDampingOverride=0.9999
    angularSpringConstant=75
    bAngularSpring=true
    bGravityOverrideInSimSpace=true
    bOverrideAngularDamping=true
    simulationSpace=BoneRelative
    alphaBoolBlend {blendOption=Linear}
    boundBone {boneName=dyn_rope_02_B}
    componentLinearAccScale=<1 1 1>
    componentLinearVelScale=<1 1 1>
    physicsBodyDefinitions [{
      sphereCollisionRadius=0
      boundBone {boneName=dyn_rope_02_B}
      boxExtents=<10 20 20>
      constraintSetup {angularLimitsMax=<0 25 25>, angularLimitsMin=<0 -25 -25>, angularTarget=<0 0 0>}
      localJointOffset=<2 0 0>
     }]
   }
   n6 = Anim Dynamics - Bone: dyn_rope_03_B(ComponentPose=@n5, Alpha=1, ExternalForce="0,0,0") {
    angularDampingOverride=0.9999
    angularSpringConstant=75
    bAngularSpring=true
    bGravityOverrideInSimSpace=true
    bOverrideAngularDamping=true
    simulationSpace=RootRelative
    alphaBoolBlend {blendOption=Linear}
    boundBone {boneName=dyn_rope_03_B}
    componentLinearAccScale=<1 1 1>
    componentLinearVelScale=<1 1 1>
    physicsBodyDefinitions [{
      sphereCollisionRadius=0
      boundBone {boneName=dyn_rope_03_B}
      boxExtents=<10 20 20>
      constraintSetup {angularLimitsMax=<0 25 25>, angularLimitsMin=<0 -25 -25>, angularTarget=<0 0 0>}
      localJointOffset=<2 0 0>
     }]
   }
   n7 = Anim Dynamics - Bone: dyn_rope_02_A(ComponentPose=@n6, Alpha=1, ExternalForce="0,0,0") {
    angularDampingOverride=0.9999
    angularSpringConstant=75
    bAngularSpring=true
    bGravityOverrideInSimSpace=true
    bOverrideAngularDamping=true
    simulationSpace=RootRelative
    alphaBoolBlend {blendOption=Linear}
    boundBone {boneName=dyn_rope_02_A}
    componentLinearAccScale=<1 1 1>
    componentLinearVelScale=<1 1 1>
    physicsBodyDefinitions [{
      sphereCollisionRadius=0
      boundBone {boneName=dyn_rope_02_A}
      boxExtents=<10 20 20>
      constraintSetup {angularLimitsMax=<0 25 25>, angularLimitsMin=<0 -25 -25>, angularTarget=<0 0 0>}
      localJointOffset=<2 0 0>
     }]
   }
   n8 = Anim Dynamics - Bone: dyn_rope_03_A(ComponentPose=@n7, Alpha=1, ExternalForce="0,0,0") {
    angularDampingOverride=0.9999
    angularSpringConstant=75
    bAngularSpring=true
    bGravityOverrideInSimSpace=true
    bOverrideAngularDamping=true
    simulationSpace=RootRelative
    alphaBoolBlend {blendOption=Linear}
    boundBone {boneName=dyn_rope_03_A}
    componentLinearAccScale=<1 1 1>
    componentLinearVelScale=<1 1 1>
    physicsBodyDefinitions [{
      sphereCollisionRadius=0
      boundBone {boneName=dyn_rope_03_A}
      boxExtents=<10 20 20>
      constraintSetup {angularLimitsMax=<0 25 25>, angularLimitsMin=<0 -25 -25>, angularTarget=<0 0 0>}
      localJointOffset=<2 0 0>
     }]
   }
   n9 = "Transform (Modify) Bone - Bone: C_jaw"(ComponentPose=@n8, Alpha=1, Translation="0,1,0", Rotation="0,0,0", Scale="1,1,1") {
    rotationMode=BMM_Additive
    rotationSpace=BCS_BoneSpace
    translationMode=BMM_Additive
    translationSpace=BCS_BoneSpace
    alphaBoolBlend {blendOption=Linear}
    boneToModify {boneName=C_jaw}
    translation {y=1}
   }
   Output Pose(Result=Blend Poses by bool(
     BlendPose_0=Component To Local(ComponentPose=@n1)
     BlendPose_1=Component To Local(ComponentPose=@n9)
     BlendTime_0=0
     BlendTime_1=0
     bActiveValue=Get GetPlayerListUsesCustomCharacterParts() -> (GetPlayerListUsesCustomCharacterParts:bool)
    ) {blendType=Linear}) {layerGroup=None}
   n10 = "Transform (Modify) Bone - Bone: dyn_beard_01"(
    ComponentPose=Local To Component(LocalPose=Copy Pose From Mesh() {bCopyCurves=true, bUseAttachedParent=true})
    Alpha=Get neck_rot_dn() -> (neck_rot_dn:double)
    Translation="0,0,0"
    Rotation="30,0,0"
    Scale="1,1,1"
   ) {
    rotationMode=BMM_Additive
    rotationSpace=BCS_BoneSpace
    scaleMode=BMM_Additive
    scaleSpace=BCS_BoneSpace
    translationMode=BMM_Additive
    translationSpace=BCS_BoneSpace
    alphaBoolBlend {blendOption=Linear}
    boneToModify {boneName=dyn_beard_01}
    rotation {pitch=30}
   }
   Save cached pose 'PreScale Cache'(Pose=Component To Local(ComponentPose=@n10)) {cacheName=PreScale Cache}
  ]}
 EventGraph {nodes [
   Event BlueprintUpdateAnimation()
   > n1 = Sequence()
   > n2 = Branch(Condition=Get bIsSkyDiving())
   > n3 = Branch(Condition=Get bIsParachuteOpened())
   n4 = BreakVector(InVec=Get PawnLastLocation() -> (PawnLastLocation:Vector))
   n5 = vector * vector(
    A:Vector="10,10,7"
    B:Vector=Make Perlin Noise Vector and Remap!(
     X:float=@n4.Z
     Y:float=@n4.Z
     Z:float=@n4.Z
     RangeOutMinX:float=-200
     RangeOutMaxX:float=200
     RangeOutMinY:float=-200
     RangeOutMaxY:float=200
     RangeOutMinZ:float=100
     RangeOutMaxZ:float=200
    ) -> (ReturnValue:Vector)
   ) -> (ReturnValue:Vector) {is=title K2Node_CommutativeAssociativeBinaryOperator}
   > Set rigidbody_external_force(rigidbody_external_force:Vector=vector * vector(A="0.2,0.2,0.2", B=@n5)) -> (Output_Get:Vector)
   Set rigidbody_external_force(execute=@n3.else, rigidbody_external_force=@n5)
   Set rigidbody_external_force(execute=@n2.else, rigidbody_external_force="0, 0, 0")
   Set neck_rot_dn(
    execute=@n1.then_1
    neck_rot_dn:double=GetAlphaFromReferenceTransformDelta(
     SkeletalMeshComponent=Get Head_Mesh() -> (Head_Mesh:Skeletal Mesh Component Object Reference)
     BoneName=C_jaw
     BaseName=spine_05
     DeltaComponentType=EulerX
     InputRange=(X=0,Y=30)
     TwistAxis="0,0,1"
    )
   ) -> (Output_Get:double)
   Event BlueprintInitializeAnimation()
   > Set GetPlayerListUsesCustomCharacterParts(GetPlayerListUsesCustomCharacterParts:bool=GetPlaylistUsesCustomCharacterParts!() -> (ReturnValue:bool)) -> (Output_Get:bool)
   > Set Skel_Mesh(Skel_Mesh=GetOwnerSkeletalMeshComponent())
   > n6 = Cast To PlayerPawn_Athena(Object=TryGetPawnOwner())
   > n7 = Set Pawn Owner(Pawn Owner:Fort Player Pawn Athena Object Reference=@n6) -> (Output_Get:Fort Player Pawn Athena Object Reference)
   > Set Head_Mesh(Head_Mesh:Skeletal Mesh Component Object Reference=GetSkeletalMeshForPartType(self=@n7, PartType=Head)) -> (Output_Get:Skeletal Mesh Component Object Reference)
   TryGetPawnOwner()
  ]}
]
variables [
 GetPlayerListUsesCustomCharacterParts {type=boolean}
 Head_Mesh {type=SkeletalMeshComponent}
 neck_rot_dn {type=number}
 neck_rot_dn_1 {type=number}
 Pawn Owner {type=FortPlayerPawnAthena}
 rigidbody_external_force {type=Vector}
 Skel_Mesh {type=SkeletalMeshComponent}
]
