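# Physics asset for the Hacivat male medium body cosmetic, providing jointed bodies for secondary motion of its coat, collar and front ropes.
# Capsules on pelvis, spine_05 and thighs form the body volumes.
# Small MSTR anchor bodies lock coat panels to the pelvis and collar flaps to spine_05, each driving one limited-swing, low-twist capsule.
# A free-rotating front knot on the pelvis carries two tightly limited, two-segment rope chains.
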PhysicsAsset /BRCosmetics/Characters/Player/Male/Medium/Bodies/M_MED_Hacivat/Mesh/Hacivat_Body_Colliders
bodies {
 dyn_coat_bk_01 [{shapeType=Capsule, length=6, radius=3, shapeName=dyn_coat_bk_01_capsule, center=<5 0 0>, rotation=<rot 90 0 0>}]
 dyn_coat_bk_MSTR [{shapeType=Sphere, radius=1, shapeName=dyn_coat_bk_MSTR_sphere}]
 dyn_coat_bk_l_01 [{shapeType=Capsule, length=6, radius=3, shapeName=dyn_coat_bk_l_01_capsule, center=<5 0 0>, rotation=<rot 90 0 0>}]
 dyn_coat_bk_l_MSTR [{shapeType=Sphere, radius=1, shapeName=dyn_coat_bk_l_MSTR_sphere}]
 dyn_coat_bk_r_01 [{shapeType=Capsule, length=6, radius=3, shapeName=dyn_coat_bk_r_01_capsule, center=<5 0 0>, rotation=<rot 90 0 0>}]
 dyn_coat_bk_r_MSTR [{shapeType=Sphere, radius=1, shapeName=dyn_coat_bk_r_MSTR_sphere}]
 dyn_coat_fr_l_01 [{shapeType=Capsule, length=8, radius=3.5, shapeName=dyn_coat_fr_l_01_capsule, center=<8 0 0>, rotation=<rot 90 0 0>}]
 dyn_coat_fr_l_MSTR [{shapeType=Capsule, length=0.505, radius=0.505, shapeName=dyn_coat_fr_l_MSTR_capsule, rotation=<rot 0 -90 0>}]
 dyn_coat_fr_r_01 [{shapeType=Capsule, length=8, radius=4, shapeName=dyn_coat_fr_r_01_capsule, center=<8 0 0>, rotation=<rot 90 0 0>}]
 dyn_coat_fr_r_MSTR [{shapeType=Sphere, radius=1, shapeName=dyn_coat_fr_r_MSTR_sphere}]
 dyn_coat_l_01 [{shapeType=Capsule, length=6, radius=2, shapeName=dyn_coat_l_01_capsule, center=<5 0 0>, rotation=<rot 90 0 0>}]
 dyn_coat_l_MSTR [{shapeType=Sphere, radius=1, shapeName=dyn_coat_l_MSTR_sphere}]
 dyn_coat_r_01 [{shapeType=Capsule, length=6, radius=2, shapeName=dyn_coat_r_01_capsule, center=<5 0 0>, rotation=<rot 90 0 0>}]
 dyn_coat_r_MSTR [{shapeType=Sphere, radius=1, shapeName=dyn_coat_r_MSTR_sphere}]
 dyn_collar_l_01 [{shapeType=Capsule, length=4.5, radius=1.5, shapeName=dyn_collar_l_01_capsule, center=<4 0 0>, rotation=<rot 90 0 0>}]
 dyn_collar_l_MSTR [{shapeType=Sphere, radius=1, shapeName=dyn_collar_l_MSTR_sphere}]
 dyn_collar_r_01 [{shapeType=Capsule, length=7, radius=1, shapeName=dyn_collar_r_01_capsule, center=<4.5 0 0>, rotation=<rot 90 0 0>}]
 dyn_collar_r_MSTR [{shapeType=Sphere, radius=1, shapeName=dyn_collar_r_MSTR_sphere}]
 knot_front [{shapeType=Capsule, length=0.505, radius=0.505, shapeName=knot_front_capsule, rotation=<rot 0 -90 0>}]
 pelvis [{shapeType=Capsule, length=10, radius=6, center=<-7 -0.9183 0>}]
 rope_front1_l [{
   shapeType=Capsule
   length=4.1387
   radius=1.7878
   shapeName=rope_front1_l_capsule
   center=<3.1382 0.0407 0.0085>
   rotation=<rot 0 90.6153 -92.4291>
  }]
 rope_front1_r [{
   shapeType=Capsule
   length=4.378
   radius=1.6089
   shapeName=rope_front1_r_capsule
   center=<-3.0443 -0.0726 -0.0721>
   rotation=<rot -0.1936 90.7093 -87.7957>
  }]
 rope_front2_l [{
   shapeType=Capsule
   length=3.6288
   radius=1.6254
   shapeName=rope_front2_l_capsule
   center=<2.6505 0.2974 -0.1852>
   rotation=<rot 0 88.0614 -87.518>
  }]
 rope_front2_r [{
   shapeType=Capsule
   length=3.8978
   radius=1.6067
   shapeName=rope_front2_r_capsule
   center=<-2.798 -0.0609 0.2773>
   rotation=<rot 0 -94.6149 -89.6716>
  }]
 spine_05 [{shapeType=Capsule, length=12, radius=11, shapeName=spine_05_capsule, center=<-2 0 0>}]
 thigh_l [{shapeType=Capsule, length=28, radius=8.5, center=<-14 0 0>, rotation=<rot 90 0 0>}]
 thigh_r [{shapeType=Capsule, length=28, radius=8.5, center=<14 0 0>, rotation=<rot 90 0 0>}]
}
constraints [
 {
  bone1Name=thigh_l
  bone2Name=pelvis
  swing1LimitDegrees=45
  swing1Motion=Free
  swing2LimitDegrees=45
  swing2Motion=Free
  twistLimitDegrees=45
  twistMotion=Free
 }
 {
  bone1Name=thigh_r
  bone2Name=pelvis
  swing1LimitDegrees=45
  swing1Motion=Free
  swing2LimitDegrees=45
  swing2Motion=Free
  twistLimitDegrees=45
  twistMotion=Free
 }
 {
  bone1Name=knot_front
  bone2Name=pelvis
  swing1LimitDegrees=45
  swing1Motion=Free
  swing2LimitDegrees=45
  swing2Motion=Free
  twistLimitDegrees=45
  twistMotion=Free
 }
 {bone1Name=rope_front1_l, bone2Name=knot_front, swing1LimitDegrees=15, swing2LimitDegrees=15, twistLimitDegrees=1}
 {bone1Name=rope_front2_l, bone2Name=rope_front1_l, swing1LimitDegrees=15, swing2LimitDegrees=15, twistLimitDegrees=1}
 {bone1Name=rope_front1_r, bone2Name=knot_front, swing1LimitDegrees=15, swing2LimitDegrees=15, twistLimitDegrees=1}
 {bone1Name=rope_front2_r, bone2Name=rope_front1_r, swing1LimitDegrees=15, swing2LimitDegrees=15, twistLimitDegrees=1}
 {
  bone1Name=dyn_coat_fr_l_MSTR
  bone2Name=pelvis
  swing1LimitDegrees=45
  swing1Motion=Locked
  swing2LimitDegrees=45
  swing2Motion=Locked
  twistLimitDegrees=45
  twistMotion=Locked
 }
 {bone1Name=dyn_coat_fr_l_01, bone2Name=dyn_coat_fr_l_MSTR, swing1LimitDegrees=22.5, swing2LimitDegrees=45, twistLimitDegrees=5}
 {
  bone1Name=dyn_coat_fr_r_MSTR
  bone2Name=pelvis
  swing1LimitDegrees=45
  swing1Motion=Locked
  swing2LimitDegrees=45
  swing2Motion=Locked
  twistLimitDegrees=45
  twistMotion=Locked
 }
 {bone1Name=dyn_coat_fr_r_01, bone2Name=dyn_coat_fr_r_MSTR, swing1LimitDegrees=22.5, swing2LimitDegrees=45, twistLimitDegrees=5}
 {
  bone1Name=dyn_coat_l_MSTR
  bone2Name=pelvis
  swing1LimitDegrees=45
  swing1Motion=Locked
  swing2LimitDegrees=45
  swing2Motion=Locked
  twistLimitDegrees=45
  twistMotion=Locked
 }
 {bone1Name=dyn_coat_l_01, bone2Name=dyn_coat_l_MSTR, swing1LimitDegrees=45, swing2LimitDegrees=45, twistLimitDegrees=5}
 {
  bone1Name=dyn_coat_r_MSTR
  bone2Name=pelvis
  swing1LimitDegrees=45
  swing1Motion=Locked
  swing2LimitDegrees=45
  swing2Motion=Locked
  twistLimitDegrees=45
  twistMotion=Locked
 }
 {bone1Name=dyn_coat_r_01, bone2Name=dyn_coat_r_MSTR, swing1LimitDegrees=45, swing2LimitDegrees=45, twistLimitDegrees=5}
 {
  bone1Name=dyn_coat_bk_l_MSTR
  bone2Name=pelvis
  swing1LimitDegrees=45
  swing1Motion=Locked
  swing2LimitDegrees=45
  swing2Motion=Locked
  twistLimitDegrees=45
  twistMotion=Locked
 }
 {bone1Name=dyn_coat_bk_l_01, bone2Name=dyn_coat_bk_l_MSTR, swing1LimitDegrees=45, swing2LimitDegrees=45, twistLimitDegrees=5}
 {
  bone1Name=dyn_coat_bk_r_MSTR
  bone2Name=pelvis
  swing1LimitDegrees=45
  swing1Motion=Locked
  swing2LimitDegrees=45
  swing2Motion=Locked
  twistLimitDegrees=45
  twistMotion=Locked
 }
 {bone1Name=dyn_coat_bk_r_01, bone2Name=dyn_coat_bk_r_MSTR, swing1LimitDegrees=45, swing2LimitDegrees=45, twistLimitDegrees=5}
 {
  bone1Name=dyn_coat_bk_MSTR
  bone2Name=pelvis
  swing1LimitDegrees=45
  swing1Motion=Locked
  swing2LimitDegrees=45
  swing2Motion=Locked
  twistLimitDegrees=45
  twistMotion=Locked
 }
 {bone1Name=dyn_coat_bk_01, bone2Name=dyn_coat_bk_MSTR, swing1LimitDegrees=45, swing2LimitDegrees=45, twistLimitDegrees=5}
 {
  bone1Name=dyn_collar_l_MSTR
  bone2Name=spine_05
  swing1LimitDegrees=45
  swing1Motion=Locked
  swing2LimitDegrees=45
  swing2Motion=Locked
  twistLimitDegrees=45
  twistMotion=Locked
 }
 {bone1Name=dyn_collar_l_01, bone2Name=dyn_collar_l_MSTR, swing1LimitDegrees=5, swing2LimitDegrees=20, twistLimitDegrees=1}
 {
  bone1Name=spine_05
  bone2Name=pelvis
  swing1LimitDegrees=45
  swing1Motion=Locked
  swing2LimitDegrees=45
  swing2Motion=Locked
  twistLimitDegrees=45
  twistMotion=Locked
 }
 {
  bone1Name=dyn_collar_r_MSTR
  bone2Name=spine_05
  swing1LimitDegrees=45
  swing1Motion=Locked
  swing2LimitDegrees=45
  swing2Motion=Locked
  twistLimitDegrees=45
  twistMotion=Locked
 }
 {bone1Name=dyn_collar_r_01, bone2Name=dyn_collar_r_MSTR, swing1LimitDegrees=35, swing2LimitDegrees=5, twistLimitDegrees=1}
]
