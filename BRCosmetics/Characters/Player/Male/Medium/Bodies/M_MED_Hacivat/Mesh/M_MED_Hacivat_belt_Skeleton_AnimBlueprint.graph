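# Post-process style AnimBlueprint adding secondary motion to the Hacivat character's belt cosmetic.
# It copies the parent mesh pose, then layers RigidBody simulation, half-weight pelvis coat constraints, and AnimDynamics on ropes, bow strings and ring.
# Left-thigh rotation drives an additive ring correction.
# Skydiving toggles cloth anim-drive stiffness with a one-shot cloth reset; the Perlin-noise dive jiggle feeds only a branch disconnected from the output pose.
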AnimBlueprint /BRCosmetics/Characters/Player/Male/Medium/Bodies/M_MED_Hacivat/Mesh/M_MED_Hacivat_belt_Skeleton_AnimBlueprint
target_skeleton=&/M_MED_Hacivat_belt_Skeleton
graphs [
 AnimGraph {nodes [
   n1 = "Anim Dynamics - Chain: rope_front1_l -> rope_front2_l"(Alpha=0, ExternalForce="0,0,0") {
    angularDampingOverride=0.96
    angularSpringConstant=500
    bAngularSpring=true
    bChain=true
    bGravityOverrideInSimSpace=true
    bOverrideAngularDamping=true
    numSolverIterationsPostUpdate=4
    numSolverIterationsPreUpdate=16
    alphaBoolBlend {blendOption=Linear}
    boundBone {boneName=rope_front1_l}
    chainEnd {boneName=rope_front2_l}
    componentAppliedLinearAccClamp=<4000 4000 4000>
    componentLinearAccScale=<1 1 1>
    componentLinearVelScale=<1 1 1>
    physicsBodyDefinitions [{
      sphereCollisionRadius=0
      boundBone {boneName=rope_front1_l}
      boxExtents=<9 18 20>
      constraintSetup {twistAxis=AxisZ, angularLimitsMax=<5 30 60>, angularLimitsMin=<-5 0 -60>, angularTarget=<1 0 0>}
      localJointOffset=<3 0 0>
     }]
   }
   n2 = "Transform (Modify) Bone - Bone: cloth_back"(ComponentPose=@n1, Alpha=1, Translation="0,0,0", Rotation=Get current_dive_jiggle() -> (current_dive_jiggle:Rotator), Scale="1,1,1") {rotationMode=BMM_Additive, rotationSpace=BCS_BoneSpace, alphaBoolBlend {blendOption=Linear}, boneToModify {boneName=cloth_back}}
   n3 = "Transform (Modify) Bone - Bone: cloth_l"(ComponentPose=@n2, Alpha=1, Translation="0,0,0", Rotation="0,0,0", Scale="1,1,1") {rotationMode=BMM_Additive, rotationSpace=BCS_BoneSpace, alphaBoolBlend {blendOption=Linear}, boneToModify {boneName=cloth_l}}
   "Transform (Modify) Bone - Bone: cloth_r"(ComponentPose=@n3, Alpha=1, Translation="0,0,0", Rotation="0,0,0", Scale="1,1,1") {rotationMode=BMM_Additive, rotationSpace=BCS_BoneSpace, alphaBoolBlend {blendOption=Linear}, boneToModify {boneName=cloth_r}}
   n4 = RigidBody(ComponentPose=Local To Component(LocalPose=Copy Pose From Mesh() {bCopyCurves=true, bUseAttachedParent=true})) {
    bClampLinearTranslationLimitToRefPose=true
    bOverrideWorldGravity=true
    overridePhysicsAsset=&/Hacivat_Body_Colliders
    simulationSpace=BaseBoneSpace
    alphaBoolBlend {blendOption=Linear}
    baseBoneRef {boneName=root}
    componentAppliedLinearAccClamp=<3000 3000 3000>
    componentLinearAccScale=<0.3 0.3 0.45>
    componentLinearVelScale=<0.15 0.15 0.225>
    overrideWorldGravity {z=-980}
   }
   n5 = "Constraint - dyn_coat_bk_r_MSTR (P)"(ComponentPose=@n4, Alpha=0.5, ConstraintWeights_0=1) {
    alphaBoolBlend {blendOption=Linear}
    boneToModify {boneName=dyn_coat_bk_r_MSTR}
    constraintSetup [{targetBone {boneName=pelvis}}]
    constraintWeights [1]
   }
   n6 = "Constraint - dyn_coat_l_MSTR (P)"(ComponentPose=@n5, Alpha=0.5, ConstraintWeights_0=1) {
    alphaBoolBlend {blendOption=Linear}
    boneToModify {boneName=dyn_coat_l_MSTR}
    constraintSetup [{targetBone {boneName=pelvis}}]
    constraintWeights [1]
   }
   n7 = "Constraint - dyn_coat_bk_l_MSTR (P)"(ComponentPose=@n6, Alpha=0.5, ConstraintWeights_0=1) {
    alphaBoolBlend {blendOption=Linear}
    boneToModify {boneName=dyn_coat_bk_l_MSTR}
    constraintSetup [{targetBone {boneName=pelvis}}]
    constraintWeights [1]
   }
   n8 = Anim Dynamics - Bone: rope_front1_l(ComponentPose=@n7, Alpha=1, ExternalForce="0,0,0") {
    angularDampingOverride=0.9999
    angularSpringConstant=75
    bAngularSpring=true
    bGravityOverrideInSimSpace=true
    bOverrideAngularDamping=true
    simulationSpace=RootRelative
    alphaBoolBlend {blendOption=Linear}
    boundBone {boneName=rope_front1_l}
    componentLinearAccScale=<1 1 1>
    componentLinearVelScale=<1 1 1>
    physicsBodyDefinitions [{
      sphereCollisionRadius=0
      boundBone {boneName=rope_front1_l}
      boxExtents=<10 20 20>
      constraintSetup {angularLimitsMax=<25 25 25>, angularLimitsMin=<-25 -2 -25>, angularTarget=<0 0 0>}
      localJointOffset=<3 0 0>
     }]
   }
   n9 = Anim Dynamics - Bone: rope_front2_l(ComponentPose=@n8, Alpha=1, ExternalForce="0,0,0") {
    angularDampingOverride=0.9999
    angularSpringConstant=75
    bAngularSpring=true
    bGravityOverrideInSimSpace=true
    bOverrideAngularDamping=true
    simulationSpace=RootRelative
    alphaBoolBlend {blendOption=Linear}
    boundBone {boneName=rope_front2_l}
    componentLinearAccScale=<1 1 1>
    componentLinearVelScale=<1 1 1>
    physicsBodyDefinitions [{
      sphereCollisionRadius=0
      boundBone {boneName=rope_front2_l}
      boxExtents=<10 20 20>
      constraintSetup {angularLimitsMax=<25 25 25>, angularLimitsMin=<-25 -2 -25>, angularTarget=<0 0 0>}
      localJointOffset=<3 0 0>
     }]
   }
   n10 = Anim Dynamics - Bone: rope_front1_r(ComponentPose=@n9, Alpha=1, ExternalForce="0,0,0") {
    angularDampingOverride=0.9999
    angularSpringConstant=75
    bAngularSpring=true
    bGravityOverrideInSimSpace=true
    bOverrideAngularDamping=true
    simulationSpace=RootRelative
    alphaBoolBlend {blendOption=Linear}
    boundBone {boneName=rope_front1_r}
    componentLinearAccScale=<1 1 1>
    componentLinearVelScale=<1 1 1>
    physicsBodyDefinitions [{
      sphereCollisionRadius=0
      boundBone {boneName=rope_front1_r}
      boxExtents=<10 20 20>
      constraintSetup {angularLimitsMax=<25 25 25>, angularLimitsMin=<-25 -2 -25>, angularTarget=<0 0 0>}
      localJointOffset=<-3 0 0>
     }]
   }
   n11 = Anim Dynamics - Bone: rope_front2_r(ComponentPose=@n10, Alpha=1, ExternalForce="0,0,0") {
    angularDampingOverride=0.9999
    angularSpringConstant=75
    bAngularSpring=true
    bGravityOverrideInSimSpace=true
    bOverrideAngularDamping=true
    simulationSpace=RootRelative
    alphaBoolBlend {blendOption=Linear}
    boundBone {boneName=rope_front2_r}
    componentLinearAccScale=<1 1 1>
    componentLinearVelScale=<1 1 1>
    physicsBodyDefinitions [{
      sphereCollisionRadius=0
      boundBone {boneName=rope_front2_r}
      boxExtents=<10 20 20>
      constraintSetup {angularLimitsMax=<25 25 25>, angularLimitsMin=<-25 -2 -25>, angularTarget=<0 0 0>}
      localJointOffset=<-3 0 0>
     }]
   }
   n12 = Anim Dynamics - Bone: Leg_L_Bow_L_String_01(ComponentPose=@n11, Alpha=1, ExternalForce="0,0,0") {
    angularDampingOverride=0.9999
    angularSpringConstant=75
    bAngularSpring=true
    bGravityOverrideInSimSpace=true
    bOverrideAngularDamping=true
    simulationSpace=RootRelative
    alphaBoolBlend {blendOption=Linear}
    boundBone {boneName=Leg_L_Bow_L_String_01}
    componentLinearAccScale=<1 1 1>
    componentLinearVelScale=<1 1 1>
    physicsBodyDefinitions [{
      sphereCollisionRadius=0
      boundBone {boneName=Leg_L_Bow_L_String_01}
      boxExtents=<10 20 20>
      constraintSetup {angularLimitsMax=<25 0 25>, angularLimitsMin=<-25 -25 -25>, angularTarget=<0 0 0>}
      localJointOffset=<1 0 -0.2>
     }]
   }
   n13 = Anim Dynamics - Bone: Leg_L_Bow_R_String_01(ComponentPose=@n12, Alpha=1, ExternalForce="0,0,0") {
    angularDampingOverride=0.9999
    angularSpringConstant=75
    bAngularSpring=true
    bGravityOverrideInSimSpace=true
    bOverrideAngularDamping=true
    simulationSpace=RootRelative
    alphaBoolBlend {blendOption=Linear}
    boundBone {boneName=Leg_L_Bow_R_String_01}
    componentLinearAccScale=<1 1 1>
    componentLinearVelScale=<1 1 1>
    physicsBodyDefinitions [{
      sphereCollisionRadius=0
      boundBone {boneName=Leg_L_Bow_R_String_01}
      boxExtents=<10 20 20>
      constraintSetup {angularLimitsMax=<25 0 25>, angularLimitsMin=<-25 -25 -25>, angularTarget=<0 0 0>}
      localJointOffset=<1 0 -0.2>
     }]
   }
   n14 = Anim Dynamics - Bone: Leg_R_Bow_L_String_01(ComponentPose=@n13, Alpha=1, ExternalForce="0,0,0") {
    angularDampingOverride=0.9999
    angularSpringConstant=75
    bAngularSpring=true
    bGravityOverrideInSimSpace=true
    bOverrideAngularDamping=true
    simulationSpace=RootRelative
    alphaBoolBlend {blendOption=Linear}
    boundBone {boneName=Leg_R_Bow_L_String_01}
    componentLinearAccScale=<1 1 1>
    componentLinearVelScale=<1 1 1>
    physicsBodyDefinitions [{
      sphereCollisionRadius=0
      boundBone {boneName=Leg_R_Bow_L_String_01}
      boxExtents=<10 20 20>
      constraintSetup {angularLimitsMax=<25 0 25>, angularLimitsMin=<-25 -25 -25>, angularTarget=<0 0 0>}
      localJointOffset=<1 0 -0.2>
     }]
   }
   n15 = Anim Dynamics - Bone: Leg_R_Bow_R_String_01(ComponentPose=@n14, Alpha=1, ExternalForce="0,0,0") {
    angularDampingOverride=0.9999
    angularSpringConstant=75
    bAngularSpring=true
    bGravityOverrideInSimSpace=true
    bOverrideAngularDamping=true
    simulationSpace=RootRelative
    alphaBoolBlend {blendOption=Linear}
    boundBone {boneName=Leg_R_Bow_R_String_01}
    componentLinearAccScale=<1 1 1>
    componentLinearVelScale=<1 1 1>
    physicsBodyDefinitions [{
      sphereCollisionRadius=0
      boundBone {boneName=Leg_R_Bow_R_String_01}
      boxExtents=<10 20 20>
      constraintSetup {angularLimitsMax=<25 0 25>, angularLimitsMin=<-25 -25 -25>, angularTarget=<0 0 0>}
      localJointOffset=<1 0 -0.2>
     }]
   }
   n16 = "Transform (Modify) Bone - Bone: ring"(ComponentPose=@n15, Alpha=Get thigh_shape_fn_l() -> (thigh_shape_fn_l:double), Translation="0,0,0", Rotation="0,10,80", Scale="1,1,1") {
    rotationMode=BMM_Additive
    rotationSpace=BCS_BoneSpace
    scaleMode=BMM_Additive
    scaleSpace=BCS_BoneSpace
    translationMode=BMM_Additive
    translationSpace=BCS_BoneSpace
    alphaBoolBlend {blendOption=Linear}
    boneToModify {boneName=ring}
    rotation {roll=80, yaw=10}
   }
   n17 = Anim Dynamics - Bone: ring(ComponentPose=@n16, Alpha=0.5, ExternalForce="0,0,0") {
    angularDampingOverride=0.8
    angularSpringConstant=5000
    bAngularSpring=true
    bGravityOverrideInSimSpace=true
    bOverrideAngularDamping=true
    gravityScale=0.8
    numSolverIterationsPostUpdate=4
    numSolverIterationsPreUpdate=16
    simulationSpace=Actor
    alphaBoolBlend {blendOption=Linear}
    boundBone {boneName=ring}
    componentLinearAccScale=<2 2>
    physicsBodyDefinitions [{
      sphereCollisionRadius=0
      boundBone {boneName=ring}
      boxExtents=<30 30 20>
      constraintSetup {angularTargetAxis=AxisY, twistAxis=AxisY, angularLimitsMax=<25 0 0>, angularLimitsMin=<-25 0 0>, angularTarget=<0 1 0>}
      localJointOffset=<0 0 -4>
     }]
   }
   Output Pose(Result=Component To Local(ComponentPose=@n17)) {layerGroup=None}
  ]}
 EventGraph {nodes [
   n1 = Event BlueprintUpdateAnimation()
   > n2 = Sequence()
   n3 = BreakVector(InVec=Get PawnLastLocation() -> (PawnLastLocation:Vector))
   > Set current_actor_z(current_actor_z:double=@n3.Z) -> (Output_Get:double)
   n4 = Make Perlin Noise and Remap!(Value:float=float + float(A=@n3.Z, B=20) {is=title K2Node_CommutativeAssociativeBinaryOperator}, RangeOutMin:float=-30, RangeOutMax:float=30) -> (ReturnValue:float)
   n5 = Make Perlin Noise and Remap(Value=@n3.Z, RangeOutMin=-30, RangeOutMax=50)
   > Set current_dive_jiggle(current_dive_jiggle:Rotator=RInterpTo!(
     Current:Rotator=Get current_dive_jiggle()
     Target:Rotator=MakeRotator(Pitch=@n5, Yaw=@n4)
     DeltaTime:float=@n1.DeltaTimeX
     InterpSpeed:float=10
    ) -> (ReturnValue:Rotator)) -> (Output_Get:Rotator)
   > n6 = Branch(Condition=Get bIsSkyDiving())
   n7 = Get drift_clothing_sim_interactor() -> (drift_clothing_sim_interactor:Clothing Simulation Interactor Nv Object Reference)
   > SetAnimDriveSpringStiffness!(execute, self:Clothing Simulation Interactor Object Reference=@n7, InStiffness:float=20) -> (then)
   > SetAnimDriveDamperStiffness!(execute, self:Clothing Simulation Interactor Nv Object Reference=@n7, InStiffness:float=10) -> (then)
   > Do Once(Reset=@n8.then, Start Closed=true)
   > ForceClothNextUpdateTeleportAndReset(self:Skeletal Mesh Component Object Reference=Get mesh_to_copy() -> (mesh_to_copy:Skeletal Mesh Component Object Reference))
   SetAnimDriveSpringStiffness(execute=@n6.else, self=@n7, InStiffness=0)
   > n8 = SetAnimDriveDamperStiffness(self=@n7, InStiffness=0)
   Set thigh_shape_fn_l(
    execute=@n2.then_1
    thigh_shape_fn_l:double=GetAlphaFromReferenceTransformDelta(SkeletalMeshComponent=Get Skel_Mesh(), BoneName=thigh_l, BaseName=pelvis, InputRange=(X=0,Y=-140), TwistAxis="0,0,1")
   ) -> (Output_Get:double)
   Event BlueprintInitializeAnimation()
   > Set Skel_Mesh(Skel_Mesh=GetOwnerSkeletalMeshComponent())
   > n9 = Cast To ClothingSimulationInteractorNv!(
    execute
    Object=GetClothingSimulationInteractor!(self:Skeletal Mesh Component Object Reference=GetPartSkeletalMeshComponent(), ClothingAssetName:Name=None) -> (ReturnValue:Clothing Simulation Interactor Object Reference)
   ) -> (then, CastFailed, AsClothing Simulation Interactor Nv:Clothing Simulation Interactor Nv Object Reference) {is=title K2Node_DynamicCast}
   > Set drift_clothing_sim_interactor(drift_clothing_sim_interactor:Clothing Simulation Interactor Nv Object Reference=@n9) -> (Output_Get:Clothing Simulation Interactor Nv Object Reference)
   TryGetPawnOwner()
  ]}
]
variables [
 current_actor_z {type=number}
 current_dive_jiggle {type=Rotator}
 drift_clothing_sim_interactor {type=ClothingSimulationInteractorNv, default=null}
 mesh_to_copy {type=SkeletalMeshComponent}
 Skel_Mesh {type=SkeletalMeshComponent}
 thigh_shape_fn_l {type=number}
]
